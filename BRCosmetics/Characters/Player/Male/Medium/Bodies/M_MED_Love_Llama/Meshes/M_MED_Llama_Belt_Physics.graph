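# Physics asset for the belt of the Love Llama male medium character body cosmetic, defining simple collision shapes on the lower body.
# A small sphere sits on the pelvis, and each thigh carries one offset, rotated capsule, with different sizes and placement on left and right.
# It has no constraints, so it likely serves as collision for the belt's secondary motion rather than as a ragdoll.
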PhysicsAsset /BRCosmetics/Characters/Player/Male/Medium/Bodies/M_MED_Love_Llama/Meshes/M_MED_Llama_Belt_Physics
bodies {
 pelvis [{shapeType=Sphere, radius=3}]
 thigh_l [{shapeType=Capsule, length=30, radius=9.25, center=<-10.1639 3.1878 3.4751>, rotation=<rot 90 -19.4817 160.4927>}]
 thigh_r [{shapeType=Capsule, length=44.75, radius=7.75, center=<13.4493 -0.754 0.5626>, rotation=<rot 90 0 179.9744>}]
}
constraints []
