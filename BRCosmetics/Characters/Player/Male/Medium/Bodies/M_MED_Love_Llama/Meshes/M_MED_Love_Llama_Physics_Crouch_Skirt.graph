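# Physics asset for the Love Llama medium male body, named for the crouch skirt, that defines simulation of the jacket's lower flaps.
# Front, side and rear jacket bone chains per leg hang from small thigh-twist anchor bodies, with nearly locked roots and looser swing down each chain.
# Capsules form the front and side segments, angled wide boxes form the rear panels, and pelvis, thigh and calf capsules supply unconstrained body shapes.
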PhysicsAsset /BRCosmetics/Characters/Player/Male/Medium/Bodies/M_MED_Love_Llama/Meshes/M_MED_Love_Llama_Physics_Crouch_Skirt
bodies {
 calf_l [{shapeType=Capsule, length=34.25, radius=9.25, center=<-19.3279 -0.0009 0>, rotation=<rot 90 0 180>}]
 calf_r [{shapeType=Capsule, length=35.25, radius=8.25, center=<19.3279 0.0009 0>, rotation=<rot 90 0 0>}]
 dyn_jacket_front_01_l [{shapeType=Capsule, length=9, radius=3, center=<5 0 0>, rotation=<rot 90 0 0>}]
 dyn_jacket_front_01_r [{shapeType=Capsule, length=9, radius=3, center=<5 0 0>, rotation=<rot 90 0 0>}]
 dyn_jacket_front_02_l [{shapeType=Capsule, length=9, radius=3, center=<5 0 0>, rotation=<rot 90 0 0>}]
 dyn_jacket_front_02_r [{shapeType=Capsule, length=9, radius=3, center=<5 0 0>, rotation=<rot 90 0 0>}]
 dyn_jacket_front_03_l [{shapeType=Capsule, length=9, radius=3, center=<5 0 0>, rotation=<rot 90 0 0>}]
 dyn_jacket_front_03_r [{shapeType=Capsule, length=9, radius=3, center=<5 0 0>, rotation=<rot 90 0 0>}]
 dyn_jacket_rear_01_l [{shapeType=Box, extentX=15, extentY=3.5, extentZ=19.5, center=<5 0 0>, rotation=<rot 0 0 15.0009>}]
 dyn_jacket_rear_01_r [{shapeType=Box, extentX=15, extentY=3.5, extentZ=20.5, center=<5 0 0>, rotation=<rot 0 0 -15.0011>}]
 dyn_jacket_rear_02_l [{shapeType=Box, extentX=15, extentY=3.5, extentZ=19.5, center=<5 0 0>, rotation=<rot 0 0 15.0009>}]
 dyn_jacket_rear_02_r [{shapeType=Box, extentX=15, extentY=3.5, extentZ=20.5, center=<5 0 0>, rotation=<rot 0 0 -19.9995>}]
 dyn_jacket_side_01_l [{shapeType=Capsule, length=9, radius=3, center=<5 0 0>, rotation=<rot 90 0 0>}]
 dyn_jacket_side_01_r [{shapeType=Capsule, length=9, radius=3, center=<5 0 0>, rotation=<rot 90 0 0>}]
 dyn_jacket_side_02_l [{shapeType=Capsule, length=9, radius=3, center=<5 0 0>, rotation=<rot 90 0 0>}]
 dyn_jacket_side_02_r [{shapeType=Capsule, length=9, radius=3, center=<5 0 0>, rotation=<rot 90 0 0>}]
 pelvis [{shapeType=Capsule, length=0.5, radius=0.5}]
 thigh_l [{shapeType=Capsule, length=24.25, radius=7.75, center=<-25 0 0>, rotation=<rot 90 0 180>}]
 thigh_r [{shapeType=Capsule, length=23, radius=9, center=<25 0 0>, rotation=<rot 90 0 0>}]
 thigh_twist_01_l [{shapeType=Box, extentX=1, extentY=1, extentZ=1}]
 thigh_twist_01_r [{shapeType=Box, extentX=1, extentY=1, extentZ=1}]
}
constraints [
 {bone1Name=dyn_jacket_side_01_l, bone2Name=thigh_twist_01_l, swing1LimitDegrees=10, swing2LimitDegrees=0, twistLimitDegrees=0}
 {bone1Name=dyn_jacket_side_02_l, bone2Name=dyn_jacket_side_01_l, swing1LimitDegrees=40, swing2LimitDegrees=30, twistLimitDegrees=10}
 {bone1Name=dyn_jacket_front_01_l, bone2Name=thigh_twist_01_l, swing1LimitDegrees=10, swing2LimitDegrees=0, twistLimitDegrees=0}
 {bone1Name=dyn_jacket_front_02_l, bone2Name=dyn_jacket_front_01_l, swing1LimitDegrees=40, swing2LimitDegrees=30, twistLimitDegrees=10}
 {bone1Name=dyn_jacket_rear_01_l, bone2Name=thigh_twist_01_l, swing1LimitDegrees=10, swing2LimitDegrees=0, twistLimitDegrees=0}
 {bone1Name=dyn_jacket_rear_02_l, bone2Name=dyn_jacket_rear_01_l, swing1LimitDegrees=40, swing2LimitDegrees=30, twistLimitDegrees=10}
 {bone1Name=dyn_jacket_side_01_r, bone2Name=thigh_twist_01_r, swing1LimitDegrees=10, swing2LimitDegrees=0, twistLimitDegrees=0}
 {bone1Name=dyn_jacket_side_02_r, bone2Name=dyn_jacket_side_01_r, swing1LimitDegrees=45, swing2LimitDegrees=45, twistLimitDegrees=10}
 {bone1Name=dyn_jacket_front_01_r, bone2Name=thigh_twist_01_r, swing1LimitDegrees=10, swing2LimitDegrees=0, twistLimitDegrees=0}
 {bone1Name=dyn_jacket_front_02_r, bone2Name=dyn_jacket_front_01_r, swing1LimitDegrees=45, swing2LimitDegrees=45, twistLimitDegrees=10}
 {bone1Name=dyn_jacket_front_03_r, bone2Name=dyn_jacket_front_02_r, swing1LimitDegrees=60, swing2LimitDegrees=60, twistLimitDegrees=10}
 {bone1Name=dyn_jacket_rear_01_r, bone2Name=thigh_twist_01_r, swing1LimitDegrees=10, swing2LimitDegrees=0, twistLimitDegrees=0}
 {bone1Name=dyn_jacket_rear_02_r, bone2Name=dyn_jacket_rear_01_r, swing1LimitDegrees=45, swing2LimitDegrees=45, twistLimitDegrees=10}
 {bone1Name=dyn_jacket_front_03_l, bone2Name=dyn_jacket_front_02_l, swing1LimitDegrees=45, swing2LimitDegrees=45, twistLimitDegrees=5}
]
