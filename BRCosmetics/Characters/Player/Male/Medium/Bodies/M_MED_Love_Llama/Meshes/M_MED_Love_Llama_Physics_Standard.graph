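# Physics asset for the male medium Love Llama character body, setting up secondary motion for its jacket and belt.
# Pelvis, lower spine, thigh and calf capsules form body volumes alongside box and capsule shapes on the dynamic bones.
# Two-segment front, side and rear jacket chains hang from tiny thigh-twist boxes, with rear panels swinging widest.
# A seven-link belt chain runs from a master bone constrained to the pelvis, with tight angular limits throughout.
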PhysicsAsset /BRCosmetics/Characters/Player/Male/Medium/Bodies/M_MED_Love_Llama/Meshes/M_MED_Love_Llama_Physics_Standard
bodies {
 calf_l [{shapeType=Capsule, length=34.25, radius=9.25, center=<-19.3279 -0.0009 0>, rotation=<rot 90 0 180>}]
 calf_r [{shapeType=Capsule, length=35.25, radius=8.25, center=<19.3279 0.0009 0>, rotation=<rot 90 0 0>}]
 dyn_belt_01 [{shapeType=Capsule, length=5, radius=3.5, center=<5 0 0.5>, rotation=<rot 0 0 90>}]
 dyn_belt_02 [{shapeType=Capsule, length=5, radius=3.5, center=<5 0 0.5>, rotation=<rot 0 0 90>}]
 dyn_belt_03 [{shapeType=Capsule, length=5, radius=3.5, center=<5 0 0.5>, rotation=<rot 0 0 90>}]
 dyn_belt_04 [{shapeType=Capsule, length=5, radius=3.5, center=<5 0 0.5>, rotation=<rot 0 0 90>}]
 dyn_belt_05 [{shapeType=Capsule, length=5, radius=3.5, center=<5 0 0.5>, rotation=<rot 0 0 90>}]
 dyn_belt_06 [{shapeType=Capsule, length=5, radius=3.5, center=<5 0 0.5>, rotation=<rot 0 0 90>}]
 dyn_belt_07 [{shapeType=Capsule, length=5, radius=3.5, center=<5 0 0.5>, rotation=<rot 0 0 90>}]
 dyn_belt_mstr [{shapeType=Capsule, length=14, radius=3, center=<9 0 0>, rotation=<rot 90 0 0>}]
 dyn_jacket_front_01_l [{shapeType=Box, extentX=15, extentY=2.5, extentZ=20, center=<-0.4838 -0.0001 5.53>, rotation=<rot 5 0.0003 5.0005>}]
 dyn_jacket_front_01_r [{shapeType=Capsule, length=9, radius=3, center=<5 0 0>, rotation=<rot 90 0 0>}]
 dyn_jacket_front_02_l [{shapeType=Box, extentX=15, extentY=5, extentZ=20, center=<2.699 0.5921 5.4772>, rotation=<rot 0.019 -0.435 11.7799>}]
 dyn_jacket_front_02_r [{shapeType=Capsule, length=9, radius=3, center=<5 0 0>, rotation=<rot 90 0 0>}]
 dyn_jacket_rear_01_l [{shapeType=Box, extentX=15, extentY=3.5, extentZ=19.5, center=<5 0 0>, rotation=<rot 0 0 15.0009>}]
 dyn_jacket_rear_01_r [{shapeType=Box, extentX=15, extentY=3.5, extentZ=20.5, center=<5 0 0>, rotation=<rot 0 0 -15.0011>}]
 dyn_jacket_rear_02_l [{shapeType=Box, extentX=15, extentY=3.9, extentZ=19.5, center=<5 0 0>, rotation=<rot 0 0 15.0009>}]
 dyn_jacket_rear_02_r [{shapeType=Box, extentX=15, extentY=3.9, extentZ=20.5, center=<5 0 0>, rotation=<rot 0 0 -19.9995>}]
 dyn_jacket_side_01_l [{shapeType=Capsule, length=9, radius=3, center=<5 0 0>, rotation=<rot 90 0 0>}]
 dyn_jacket_side_01_r [{shapeType=Capsule, length=9, radius=3, center=<5 0 0>, rotation=<rot 90 0 0>}]
 dyn_jacket_side_02_l [{shapeType=Capsule, length=9, radius=3, center=<5 0 0>, rotation=<rot 90 0 0>}]
 dyn_jacket_side_02_r [{shapeType=Capsule, length=9, radius=3, center=<5 0 0>, rotation=<rot 90 0 0>}]
 pelvis [{shapeType=Capsule, length=0.5, radius=21.25, center=<4.0601 -10.6714 0>}]
 spine_01 [{shapeType=Capsule, length=0.5, radius=21.25, center=<-3.159 8.3218 0>}]
 thigh_l [
  {shapeType=Capsule, length=24.25, radius=10.25, center=<-24.9999 1.0036 6.7551>, rotation=<rot 90 0 -179.9999>}
  {shapeType=Capsule, length=24.25, radius=9.5, center=<-25 0 0>, rotation=<rot 90 0 180>}
 ]
 thigh_r [{shapeType=Capsule, length=23, radius=9, center=<25 0 0>, rotation=<rot 90 0 0>}]
 thigh_twist_01_l [{shapeType=Box, extentX=1, extentY=1, extentZ=1}]
 thigh_twist_01_r [{shapeType=Box, extentX=1, extentY=1, extentZ=1}]
}
constraints [
 {bone1Name=dyn_jacket_side_01_l, bone2Name=thigh_twist_01_l, swing1LimitDegrees=45, swing2LimitDegrees=5, twistLimitDegrees=5}
 {bone1Name=dyn_jacket_side_02_l, bone2Name=dyn_jacket_side_01_l, swing1LimitDegrees=45, swing2LimitDegrees=45, twistLimitDegrees=5}
 {bone1Name=dyn_jacket_front_01_l, bone2Name=thigh_twist_01_l, swing1LimitDegrees=15, swing2LimitDegrees=5, twistLimitDegrees=5}
 {bone1Name=dyn_jacket_front_02_l, bone2Name=dyn_jacket_front_01_l, swing1LimitDegrees=45, swing2LimitDegrees=30, twistLimitDegrees=5}
 {bone1Name=dyn_jacket_rear_01_l, bone2Name=thigh_twist_01_l, swing1LimitDegrees=80, swing2LimitDegrees=5, twistLimitDegrees=5}
 {bone1Name=dyn_jacket_rear_02_l, bone2Name=dyn_jacket_rear_01_l, swing1LimitDegrees=80, swing2LimitDegrees=15, twistLimitDegrees=5}
 {bone1Name=dyn_jacket_side_01_r, bone2Name=thigh_twist_01_r, swing1LimitDegrees=45, swing2LimitDegrees=5, twistLimitDegrees=5}
 {bone1Name=dyn_jacket_side_02_r, bone2Name=dyn_jacket_side_01_r, swing1LimitDegrees=45, swing2LimitDegrees=45, twistLimitDegrees=5}
 {bone1Name=dyn_jacket_front_01_r, bone2Name=thigh_twist_01_r, swing1LimitDegrees=20, swing2LimitDegrees=5, twistLimitDegrees=5}
 {bone1Name=dyn_jacket_front_02_r, bone2Name=dyn_jacket_front_01_r, swing1LimitDegrees=45, swing2LimitDegrees=30, twistLimitDegrees=5}
 {bone1Name=dyn_jacket_rear_01_r, bone2Name=thigh_twist_01_r, swing1LimitDegrees=80, swing2LimitDegrees=5, twistLimitDegrees=5}
 {bone1Name=dyn_jacket_rear_02_r, bone2Name=dyn_jacket_rear_01_r, swing1LimitDegrees=80, swing2LimitDegrees=15, twistLimitDegrees=5}
 {bone1Name=dyn_belt_01, bone2Name=dyn_belt_mstr, swing1LimitDegrees=11.25, swing2LimitDegrees=11.25, twistLimitDegrees=5}
 {bone1Name=dyn_belt_02, bone2Name=dyn_belt_01, swing1LimitDegrees=11.25, swing2LimitDegrees=11.25, twistLimitDegrees=5}
 {bone1Name=dyn_belt_03, bone2Name=dyn_belt_02, swing1LimitDegrees=11.25, swing2LimitDegrees=11.25, twistLimitDegrees=5}
 {bone1Name=dyn_belt_04, bone2Name=dyn_belt_03, swing1LimitDegrees=11.25, swing2LimitDegrees=11.25, twistLimitDegrees=5}
 {bone1Name=dyn_belt_05, bone2Name=dyn_belt_04, swing1LimitDegrees=11.25, swing2LimitDegrees=11.25, twistLimitDegrees=5}
 {bone1Name=dyn_belt_06, bone2Name=dyn_belt_05, swing1LimitDegrees=11.25, swing2LimitDegrees=11.25, twistLimitDegrees=5}
 {bone1Name=dyn_belt_07, bone2Name=dyn_belt_06, swing1LimitDegrees=11.25, swing2LimitDegrees=11.25, twistLimitDegrees=5}
 {bone1Name=dyn_belt_mstr, bone2Name=pelvis, swing1LimitDegrees=11.25, swing2LimitDegrees=11.25, twistLimitDegrees=5}
]
