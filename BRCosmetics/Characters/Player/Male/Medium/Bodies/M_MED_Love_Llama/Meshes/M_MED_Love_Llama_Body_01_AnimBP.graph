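# Post-process-style cosmetic AnimBP for the Love Llama male body, animating collar, jacket and belt over the pose copied from the parent mesh.
# A collar Pose Driver feeds a cached pose into RigidBody simulations chosen by front-end and crouch state, with speed-weighted belt reset while crouching.
# When physics is off or LOD exceeds 1, calf rotation constraints drive jacket bones instead.
# Crouch, skydive and parachute events reset dynamics; several belt-constraint chains are unconnected.
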AnimBlueprint /BRCosmetics/Characters/Player/Male/Medium/Bodies/M_MED_Love_Llama/Meshes/M_MED_Love_Llama_Body_01_AnimBP
target_skeleton=&/M_MED_Love_Llama_Body_01_Skeleton
graphs [
 AnimGraph {nodes [
   n1 = Copy Bone Delta - Source Bone: thigh_l - Target Bone: dyn_belt_mstr!(ComponentPose:Component Space Pose Link Structure=Local To Component(LocalPose=Use cached pose 'crouchCache'()), Alpha:float=0.25) -> (Pose:Component Space Pose Link Structure) {
    is=title AnimGraphNode_CopyBoneDelta
    bCopyRotation=true
    alphaBoolBlend {blendOption=Linear}
    sourceBone {boneName=thigh_l}
    targetBone {boneName=dyn_belt_mstr}
   }
   Blend Poses by bool(
    BlendPose_0=Layered blend per bone(BlendPoses_0=Sequence Player 'M_MED_Love_Llama_Belt_RBF'() {sequence=&/M_MED_Love_Llama_Belt_RBF}) {layerSetup [{branchFilters [{blendDepth=1, boneName=dyn_belt_mstr}]}]}
    BlendPose_1=Component To Local(ComponentPose=@n1)
    bActiveValue=Get bIsCrouching()
   ) {blendType=Linear}
   Component To Local()
   n2 = "Constraint - dyn_belt_r_01 (R)"(ComponentPose=Local To Component(), Alpha=1, ConstraintWeights_0=1) {
    alphaBoolBlend {blendOption=Linear}
    boneToModify {boneName=dyn_belt_r_01}
    constraintSetup [{transformType=Rotation, targetBone {boneName=dyn_jacket_front_01_l}}]
    constraintWeights [1]
   }
   n3 = "Constraint - dyn_belt_l_01 (R)"(ComponentPose=@n2, Alpha=1, ConstraintWeights_0=1) {
    alphaBoolBlend {blendOption=Linear}
    boneToModify {boneName=dyn_belt_l_01}
    constraintSetup [{transformType=Rotation, targetBone {boneName=dyn_jacket_front_01_l}}]
    constraintWeights [1]
   }
   n4 = "Constraint - dyn_belt_r_02 (R)"(ComponentPose=@n3, Alpha=1, ConstraintWeights_0=1) {
    alphaBoolBlend {blendOption=Linear}
    boneToModify {boneName=dyn_belt_r_02}
    constraintSetup [{transformType=Rotation, targetBone {boneName=dyn_jacket_front_02_l}}]
    constraintWeights [1]
   }
   n5 = "Constraint - dyn_belt_l_02 (R)"(ComponentPose=@n4, Alpha=1, ConstraintWeights_0=1) {
    alphaBoolBlend {blendOption=Linear}
    boneToModify {boneName=dyn_belt_l_02}
    constraintSetup [{transformType=Rotation, targetBone {boneName=dyn_jacket_front_02_l}}]
    constraintWeights [1]
   }
   n6 = "Constraint - dyn_belt_r_03 (R)"(ComponentPose=@n5, Alpha=1, ConstraintWeights_0=1) {
    alphaBoolBlend {blendOption=Linear}
    boneToModify {boneName=dyn_belt_r_03}
    constraintSetup [{transformType=Rotation, targetBone {boneName=dyn_jacket_front_03_l}}]
    constraintWeights [1]
   }
   n7 = "Constraint - dyn_belt_l_03 (R)"(ComponentPose=@n6, Alpha=1, ConstraintWeights_0=1) {
    alphaBoolBlend {blendOption=Linear}
    boneToModify {boneName=dyn_belt_l_03}
    constraintSetup [{transformType=Rotation, targetBone {boneName=dyn_jacket_front_03_l}}]
    constraintWeights [1]
   }
   Component To Local(ComponentPose=@n7)
   n8 = "Constraint - dyn_belt_r_01 (RRR)"!(ComponentPose, Alpha=1, ConstraintWeights_0=0.2986, ConstraintWeights_1:float=0.0603, ConstraintWeights_2:float=0.6411) -> (Pose) {
    alphaBoolBlend {blendOption=Linear}
    boneToModify {boneName=dyn_belt_r_01}
    constraintSetup [
     {transformType=Rotation, targetBone {boneName=dyn_jacket_front_01_l}}
     {transformType=Rotation, targetBone {boneName=spine_01}}
     {transformType=Rotation, targetBone {boneName=pelvis}}
    ]
    constraintWeights [1,1,1]
   }
   n9 = "Constraint - dyn_belt_l_01 (RRR)"!(ComponentPose=@n8, Alpha=1, ConstraintWeights_0=0.2986, ConstraintWeights_1:float=0.0603, ConstraintWeights_2:float=0.6411) -> (Pose) {
    alphaBoolBlend {blendOption=Linear}
    boneToModify {boneName=dyn_belt_l_01}
    constraintSetup [
     {transformType=Rotation, targetBone {boneName=dyn_jacket_front_01_l}}
     {transformType=Rotation, targetBone {boneName=spine_01}}
     {transformType=Rotation, targetBone {boneName=pelvis}}
    ]
    constraintWeights [1,1,1]
   }
   n10 = "Constraint - dyn_belt_r_02 (RRRR)"!(
    ComponentPose=@n9
    Alpha=1
    ConstraintWeights_0=0.8477
    ConstraintWeights_1:float=0.0021
    ConstraintWeights_2:float=0.0369
    ConstraintWeights_3:float=0.1134
   ) -> (Pose) {
    alphaBoolBlend {blendOption=Linear}
    boneToModify {boneName=dyn_belt_r_02}
    constraintSetup [
     {transformType=Rotation, targetBone {boneName=dyn_jacket_front_01_l}}
     {transformType=Rotation, targetBone {boneName=spine_01}}
     {transformType=Rotation, targetBone {boneName=thigh_l}}
     {transformType=Rotation, targetBone {boneName=pelvis}}
    ]
    constraintWeights [1,1,1,1]
   }
   n11 = "Constraint - dyn_belt_l_02 (RRRR)"!(
    ComponentPose=@n10
    Alpha=1
    ConstraintWeights_0=0.8477
    ConstraintWeights_1:float=0.0021
    ConstraintWeights_2:float=0.0369
    ConstraintWeights_3:float=0.1134
   ) -> (Pose) {
    alphaBoolBlend {blendOption=Linear}
    boneToModify {boneName=dyn_belt_l_02}
    constraintSetup [
     {transformType=Rotation, targetBone {boneName=dyn_jacket_front_01_l}}
     {transformType=Rotation, targetBone {boneName=spine_01}}
     {transformType=Rotation, targetBone {boneName=thigh_l}}
     {transformType=Rotation, targetBone {boneName=pelvis}}
    ]
    constraintWeights [1,1,1,1]
   }
   n12 = "Constraint - dyn_belt_r_03 (RRRRRRRR)"!(
    ComponentPose=@n11
    Alpha=1
    ConstraintWeights_0=0.0012
    ConstraintWeights_1:float=0.433
    ConstraintWeights_2:float=0.2555
    ConstraintWeights_3:float=0.1727
    ConstraintWeights_4:float=0.0751
    ConstraintWeights_5:float=0.0146
    ConstraintWeights_6:float=0.0263
    ConstraintWeights_7:float=0.0216
   ) -> (Pose) {
    alphaBoolBlend {blendOption=Linear}
    boneToModify {boneName=dyn_belt_r_03}
    constraintSetup [
     {transformType=Rotation, targetBone {boneName=dyn_jacket_rear_01_l}}
     {transformType=Rotation, targetBone {boneName=dyn_jacket_front_03_l}}
     {transformType=Rotation, targetBone {boneName=dyn_jacket_front_01_l}}
     {transformType=Rotation, targetBone {boneName=dyn_jacket_front_02_l}}
     {transformType=Rotation, targetBone {boneName=dyn_jacket_front_04_l}}
     {transformType=Rotation, targetBone {boneName=dyn_jacket_side_02_l}}
     {transformType=Rotation, targetBone {boneName=pelvis}}
     {transformType=Rotation, targetBone {boneName=dyn_jacket_side_01_l}}
    ]
    constraintWeights [1,1,1,1,1,1,1,1]
   }
   n13 = "Constraint - dyn_belt_l_03 (RRRRRRRR)"!(
    ComponentPose=@n12
    Alpha=1
    ConstraintWeights_0=0.0012
    ConstraintWeights_1:float=0.433
    ConstraintWeights_2:float=0.2555
    ConstraintWeights_3:float=0.1727
    ConstraintWeights_4:float=0.0751
    ConstraintWeights_5:float=0.0146
    ConstraintWeights_6:float=0.0263
    ConstraintWeights_7:float=0.0216
   ) -> (Pose) {
    alphaBoolBlend {blendOption=Linear}
    boneToModify {boneName=dyn_belt_l_03}
    constraintSetup [
     {transformType=Rotation, targetBone {boneName=dyn_jacket_rear_01_l}}
     {transformType=Rotation, targetBone {boneName=dyn_jacket_front_03_l}}
     {transformType=Rotation, targetBone {boneName=dyn_jacket_front_01_l}}
     {transformType=Rotation, targetBone {boneName=dyn_jacket_front_02_l}}
     {transformType=Rotation, targetBone {boneName=dyn_jacket_front_04_l}}
     {transformType=Rotation, targetBone {boneName=dyn_jacket_side_02_l}}
     {transformType=Rotation, targetBone {boneName=pelvis}}
     {transformType=Rotation, targetBone {boneName=dyn_jacket_side_01_l}}
    ]
    constraintWeights [1,1,1,1,1,1,1,1]
   }
   n14 = "Constraint - dyn_belt_r_04 (RRRRRRRRR)"!(
    ComponentPose=@n13
    Alpha=1
    ConstraintWeights_0=0.3512
    ConstraintWeights_1:float=0.1674
    ConstraintWeights_2:float=0.0153
    ConstraintWeights_3:float=0.0203
    ConstraintWeights_4:float=0.0109
    ConstraintWeights_5:float=0.2752
    ConstraintWeights_6:float=0.0171
    ConstraintWeights_7:float=0.0182
    ConstraintWeights_8:float=0.1244
   ) -> (Pose) {
    alphaBoolBlend {blendOption=Linear}
    boneToModify {boneName=dyn_belt_r_04}
    constraintSetup [
     {transformType=Rotation, targetBone {boneName=dyn_jacket_front_03_l}}
     {transformType=Rotation, targetBone {boneName=dyn_jacket_front_01_l}}
     {transformType=Rotation, targetBone {boneName=dyn_jacket_side_04_l}}
     {transformType=Rotation, targetBone {boneName=dyn_jacket_side_01_l}}
     {transformType=Rotation, targetBone {boneName=dyn_jacket_side_03_l}}
     {transformType=Rotation, targetBone {boneName=dyn_jacket_front_04_l}}
     {transformType=Rotation, targetBone {boneName=dyn_jacket_side_02_l}}
     {transformType=Rotation, targetBone {boneName=pelvis}}
     {transformType=Rotation, targetBone {boneName=dyn_jacket_front_02_l}}
    ]
    constraintWeights [1,1,1,1,1,1,1,1,1]
   }
   "Constraint - dyn_belt_l_04 (RRRRRRRRR)"!(
    ComponentPose=@n14
    Alpha=1
    ConstraintWeights_0=0.3512
    ConstraintWeights_1:float=0.1674
    ConstraintWeights_2:float=0.0153
    ConstraintWeights_3:float=0.0203
    ConstraintWeights_4:float=0.0109
    ConstraintWeights_5:float=0.2752
    ConstraintWeights_6:float=0.0171
    ConstraintWeights_7:float=0.0182
    ConstraintWeights_8:float=0.1244
   ) -> (Pose) {
    alphaBoolBlend {blendOption=Linear}
    boneToModify {boneName=dyn_belt_l_04}
    constraintSetup [
     {transformType=Rotation, targetBone {boneName=dyn_jacket_front_03_l}}
     {transformType=Rotation, targetBone {boneName=dyn_jacket_front_01_l}}
     {transformType=Rotation, targetBone {boneName=dyn_jacket_side_04_l}}
     {transformType=Rotation, targetBone {boneName=dyn_jacket_side_01_l}}
     {transformType=Rotation, targetBone {boneName=dyn_jacket_side_03_l}}
     {transformType=Rotation, targetBone {boneName=dyn_jacket_front_04_l}}
     {transformType=Rotation, targetBone {boneName=dyn_jacket_side_02_l}}
     {transformType=Rotation, targetBone {boneName=pelvis}}
     {transformType=Rotation, targetBone {boneName=dyn_jacket_front_02_l}}
    ]
    constraintWeights [1,1,1,1,1,1,1,1,1]
   }
   Layered blend per bone!(BasePose, BlendPoses_0, BlendPoses_1:Pose Link Structure, BlendWeights_0, BlendWeights_1:float=1) -> (Pose) {
    blendPoses [{linkId=-1, sourceLinkId=-1}, {linkId=-1, sourceLinkId=-1}]
    blendWeights [1,1]
    layerSetup [{branchFilters [{blendDepth=1, boneName=const_belt_mstr}]}, {branchFilters [{blendDepth=1, boneName=dyn_belt_mstr}]}]
   }
   n15 = RigidBody!(
    ComponentPose=Local To Component(LocalPose=Use cached pose 'CPFM Cache'())
    Alpha
    OverrideWorldGravity:Vector="0,0,-980"
    ExternalForce
    ComponentLinearAccScale:Vector="0.2,0.2,0.2"
    ComponentLinearVelScale:Vector="0.3,0.3,0.3"
   ) -> (Pose) {
    bClampLinearTranslationLimitToRefPose=true
    bOverrideWorldGravity=true
    overridePhysicsAsset=&/M_MED_Love_Llama_Physics_Standard
    simulationSpace=BaseBoneSpace
    alphaBoolBlend {blendOption=Linear}
    baseBoneRef {boneName=pelvis}
    componentAppliedLinearAccClamp=<1000 1000 1000>
    componentLinearAccScale=<0.1 0.1 0.1>
    componentLinearVelScale=<1 1 1>
    overrideWorldGravity {z=-715}
   }
   n16 = Get PawnSpeed2D()
   n17 = RigidBody(
    ComponentPose=Local To Component(LocalPose=Use cached pose 'CPFM Cache'())
    OverrideWorldGravity="0,0,-980"
    ExternalForce=Break Fort Anim Input Skydiving External Force(FortAnimInput_SkydivingExternalForce=Get SkydivingExternalForceInput()) -> (
     bUseSkydivingVectorForce=~
     SkydivingVectorForce=~
     HeadToPelvisDirection=~
     FloatingMultiplier=~
     FloatingAdditive=~
     DivingMultiplier=~
     DivingAdditive=~
     ParachutingMultiplier=~
     ParachutingAdditive=~
    ).SkydivingVectorForce
    ComponentLinearAccScale="0.2,0.2,1"
    ComponentLinearVelScale=SelectVector(A="0, 0, 0", B="1,1,1", bPickA="float > float"(A=@n16, B=0) {is=title K2Node_CallFunction})
   ) {
    bClampLinearTranslationLimitToRefPose=true
    bOverrideWorldGravity=true
    overridePhysicsAsset=&M_MED_Love_Llama_Physics_Standard
    simulationSpace=BaseBoneSpace
    alphaBoolBlend {blendOption=Linear}
    baseBoneRef {boneName=pelvis}
    componentAppliedLinearAccClamp=<1000 1000 1000>
    componentLinearAccScale=<0.1 0.1 0.1>
    componentLinearVelScale=<1 1 1>
    overrideWorldGravity {z=-715}
   }
   n18 = Blend Poses by bool(BlendPose_0=Component To Local(ComponentPose=@n15), BlendPose_1=Component To Local(ComponentPose=@n17), bActiveValue=Get bIsInFrontEnd()) {blendType=Linear}
   n19 = Break Fort Anim Input Skydiving External Force(FortAnimInput_SkydivingExternalForce=Get SkydivingExternalForceInput()) -> (
    bUseSkydivingVectorForce=~
    SkydivingVectorForce=~
    HeadToPelvisDirection=~
    FloatingMultiplier=~
    FloatingAdditive=~
    DivingMultiplier=~
    DivingAdditive=~
    ParachutingMultiplier=~
    ParachutingAdditive=~
   )
   n20 = RigidBody!(ComponentPose=Local To Component(LocalPose=Use cached pose 'CPFM Cache'()), Alpha=0.35, ExternalForce=@n19.SkydivingVectorForce) -> (Pose) {
    bClampLinearTranslationLimitToRefPose=true
    bOverrideWorldGravity=true
    overridePhysicsAsset=&/M_MED_Love_Llama_Physics_Crouch
    simulationSpace=BaseBoneSpace
    alphaBoolBlend {blendOption=Linear}
    baseBoneRef {boneName=pelvis}
    componentAppliedLinearAccClamp=<1000 1000 1000>
    componentLinearAccScale=<1 1 1>
    componentLinearVelScale=<1 1 1>
    overrideWorldGravity {z=-980}
   }
   n21 = RigidBody(ComponentPose=@n20, ExternalForce=@n19.SkydivingVectorForce) {
    bClampLinearTranslationLimitToRefPose=true
    bOverrideWorldGravity=true
    overridePhysicsAsset=&/M_MED_Love_Llama_Physics_Crouch_Skirt
    simulationSpace=BaseBoneSpace
    alphaBoolBlend {blendOption=Linear}
    baseBoneRef {boneName=pelvis}
    componentAppliedLinearAccClamp=<1000 1000 1000>
    componentLinearAccScale=<1 1 1>
    componentLinearVelScale=<1 1 1>
    overrideWorldGravity {z=-980}
   }
   n22 = Layered blend per bone(
    BasePose=Use cached pose 'CPFM Cache'()
    BlendPoses_0=Sequence Player 'M_MED_Love_Llama_beltReset'() {sequence=&/M_MED_Love_Llama_beltReset}
    BlendPoses_1=Component To Local(ComponentPose=@n21)
    BlendWeights_0=MapRangeClamped(Value=Get PawnSpeed2D(), InRangeA=0, InRangeB=420)
    BlendWeights_1=1
   ) {
    blendPoses [{linkId=-1, sourceLinkId=-1}, {linkId=-1, sourceLinkId=-1}]
    blendWeights [1,1]
    layerSetup [{branchFilters [{blendDepth=1, boneName=const_belt_mstr}]}, {branchFilters [{blendDepth=1, boneName=dyn_belt_mstr}]}]
   }
   n23 = Get bIsCrouching()
   n24 = "Constraint - dyn_jacket_front_04_l (R)"(ComponentPose=Local To Component(LocalPose=Use cached pose 'CPFM Cache'()), Alpha=1, ConstraintWeights_0=1) {
    alphaBoolBlend {blendOption=Linear}
    boneToModify {boneName=dyn_jacket_front_04_l}
    constraintSetup [{transformType=Rotation, targetBone {boneName=calf_l}}]
    constraintWeights [1]
   }
   n25 = "Constraint - dyn_jacket_front_04_r (R)"(ComponentPose=@n24, Alpha=1, ConstraintWeights_0=1) {
    alphaBoolBlend {blendOption=Linear}
    boneToModify {boneName=dyn_jacket_front_04_r}
    constraintSetup [{transformType=Rotation, targetBone {boneName=calf_r}}]
    constraintWeights [1]
   }
   n26 = "Constraint - dyn_jacket_rear_03_l (R)"(ComponentPose=@n25, Alpha=1, ConstraintWeights_0=1) {
    alphaBoolBlend {blendOption=Linear}
    boneToModify {boneName=dyn_jacket_rear_03_l}
    constraintSetup [{transformType=Rotation, targetBone {boneName=calf_l}}]
    constraintWeights [1]
   }
   n27 = "Constraint - dyn_jacket_rear_03_r (R)"(ComponentPose=@n26, Alpha=1, ConstraintWeights_0=1) {
    alphaBoolBlend {blendOption=Linear}
    boneToModify {boneName=dyn_jacket_rear_03_r}
    constraintSetup [{transformType=Rotation, targetBone {boneName=calf_r}}]
    constraintWeights [1]
   }
   n28 = "Constraint - dyn_jacket_side_03_l (R)"(ComponentPose=@n27, Alpha=1, ConstraintWeights_0=1) {
    alphaBoolBlend {blendOption=Linear}
    boneToModify {boneName=dyn_jacket_side_03_l}
    constraintSetup [{transformType=Rotation, targetBone {boneName=calf_l}}]
    constraintWeights [1]
   }
   n29 = "Constraint - dyn_jacket_side_03_r (R)"(ComponentPose=@n28, Alpha=1, ConstraintWeights_0=1) {
    alphaBoolBlend {blendOption=Linear}
    boneToModify {boneName=dyn_jacket_side_03_r}
    constraintSetup [{transformType=Rotation, targetBone {boneName=calf_r}}]
    constraintWeights [1]
   }
   n30 = "integer > integer"(A=Get CurrentLOD(), B=1) {is=title K2Node_CallFunction}
   n31 = NOT Boolean(A=Get bRigidBodyEnabled())
   Output Pose(Result=Blend Poses by bool(
     BlendPose_0=Component To Local(ComponentPose=@n29)
     BlendPose_1=Blend Poses by bool(BlendPose_0=@n22, BlendPose_1=@n18, BlendTime_0=0, BlendTime_1=1, bActiveValue=@n23) {blendType=Linear}
     bActiveValue=OR Boolean(A=@n31, B=@n30)
    ) {blendType=Linear}) {layerGroup=None}
   n32 = Pose Driver - Source: upperarm_l!(SourcePose:Pose Link Structure=Copy Pose From Mesh() {bUseAttachedParent=true}) -> (Pose:Pose Link Structure) {
    is=title AnimGraphNode_PoseDriver
    poseAsset=&/M_MED_Love_Llama_RBF_PoseAsset
    onlyDriveBones [{boneName=dyn_collar}]
    poseTargets [
     {
      bApplyCustomCurve=false
      bIsHidden=false
      distanceMethod=DefaultMethod
      drivenName=Pose_0
      functionType=DefaultFunction
      targetScale=1.6
      boneTransforms [{targetRotation=<rot 0 0 0>, targetTranslation=<0 0 0>}, {targetRotation=<rot 0 0 0>, targetTranslation=<0 0 0>}]
      customCurve {defaultValue=3.4028234663852886e+38, postInfinityExtrap=RCCE_Constant, preInfinityExtrap=RCCE_Constant, keys []}
      targetRotation=<rot 0 0 0>
     }
     {
      bApplyCustomCurve=false
      bIsHidden=false
      distanceMethod=DefaultMethod
      drivenName=Pose_1
      functionType=DefaultFunction
      targetScale=1
      boneTransforms [
       {targetRotation=<rot 10.946 56.0172 -99.8288>, targetTranslation=<15.5111 0 0>}
       {targetRotation=<rot 34.3413 94.8833 -85.4429>, targetTranslation=<-15.5111 0 0>}
      ]
      customCurve {defaultValue=3.4028234663852886e+38, postInfinityExtrap=RCCE_Constant, preInfinityExtrap=RCCE_Constant, keys []}
      targetRotation=<rot 0 0 0>
     }
    ]
    rBFParams {distanceMethod=Quaternion, radius=91.5278}
    sourceBones [{boneName=upperarm_l}, {boneName=upperarm_r}]
   }
   Save cached pose 'CPFM Cache'(Pose=@n32) {cacheName=CPFM Cache}
   Save cached pose 'crouchCache'() {cacheName=crouchCache}
   Sequence Player 'M_MED_Love_Llama_RBF'() {playRate=0.01, sequence=&/M_MED_Love_Llama_RBF}
   Use cached pose 'crouchCache'()
  ]}
 EventGraph {nodes [
   Event OnParachuteOpened()
   > n1 = Delay(Duration=1.5)
   n2 = ResetDynamics(execute=[@n1.then, @n6.then, @n3.then, @n4.then], InTeleportType=TeleportPhysics)
   ResetDynamics(execute=[@n2.then, @n9.OutputPin], InTeleportType=ResetPhysics)
   n3 = Event OnBeginSkydiving()
   n4 = Event OnEndSkydiving()
   n5 = Event OnEndCrouching()
   n6 = Delay(execute=[@n5.then, @n8.then], Duration=0.125)
   Event OnBeginTargeting() {is=title K2Node_Event}
   n7 = Get bIsCrouching()
   > Branch(Condition=AND Boolean(A=Get bIsTargeting(), B=NOT Boolean(A=@n7)))
   n8 = Event OnBeginCrouching()
   n9 = Reroute Node()
   TryGetPawnOwner()
   ResetDynamics(InTeleportType=TeleportPhysics)
   Event BlueprintUpdateAnimation()
  ]}
]
variables []
